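# Physics asset for the Tower Sentinel medium male player body cosmetic mesh.
# It holds a single body on spine_05, a rotated capsule offset slightly from the bone.
# With no constraints and no other bodies, it gives only coarse upper-torso collision or bounds rather than a jointed ragdoll.
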PhysicsAsset /BRCosmetics/Characters/Player/Male/Medium/Bodies/M_MED_Tower_Sentinel/Meshes/M_MED_Tower_Sentinel_Physics
bodies {spine_05 [{shapeType=Capsule, length=14.7403, radius=12.6533, center=<4.6847 0.6253 -0.2959>, rotation=<rot 0 -172.3584 -89.2095>}]}
constraints []
