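# Animation Blueprint for the Tower Sentinel cosmetic body, copying the parent mesh's pose and adding corrective and secondary motion.
# Constraints and additive offsets adjust thigh and upper-arm helper bones, weighted by per-frame joint rotation deltas.
# AmbientMotion sine waves float three pumice bones in unison; computed rotations feed nodes whose rotation mode stays at default.
# A pelvis-space RigidBody node finishes with physics simulation.
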AnimBlueprint /BRCosmetics/Characters/Player/Male/Medium/Bodies/M_MED_Tower_Sentinel/Meshes/M_MED_Tower_Sentinel_AnimBP
target_skeleton=&/M_MED_Tower_Sentinel_Skeleton
graphs [
 AmbientMotion {nodes [
   n1 = AmbientMotion!() -> (
    then
    Period:Vector="1.5,1,2"
    PeriodMult:double=1
    TimeOffset:double=~
    Magnitude:Vector="0.3,0.56,0.22"
    MagnitudeMult:double=1
    RotateMult:double=1
   ) {is=title K2Node_FunctionEntry}
   n2 = BreakVector(InVec=@n1.Magnitude)
   n3 = BreakVector(InVec=@n1)
   n4 = GetGameTimeInSeconds()
   n5 = float * float(A=float * float(A=@n1.PeriodMult, B=@n3.Y), B=@n4)
   n6 = float * float(A="Sin (Radians)"(A=float + float(A=@n5, B=@n1.TimeOffset) {is=title K2Node_CommutativeAssociativeBinaryOperator}), B=@n2.Y)
   n7 = float * float(A=float * float(A=@n1.PeriodMult, B=@n3.Z), B=@n4)
   n8 = float * float(A="Sin (Radians)"(A=float + float(A=@n7, B=@n1.TimeOffset)), B=@n2.Z)
   n9 = float * float(A=float * float(A=@n1.PeriodMult, B=@n3), B=@n4)
   n10 = float * float(A="Sin (Radians)"(A=float + float(A=@n9, B=@n1.TimeOffset)), B=@n2)
   n11 = MakeRotator(Roll=@n10, Pitch=@n6, Yaw=@n8)
   > Return Node!(
    execute
    VectorOut:Vector=vector * float(A=MakeVector(X=@n10, Y=@n6, Z=@n8), B=@n1.MagnitudeMult)
    RotateOut:Rotator="Scale Rotator (Float)"!(A:Rotator="Scale Rotator (Float)"(A=@n11, B=@n1.MagnitudeMult), B:float=@n1.RotateMult) -> (ReturnValue:Rotator)
   )
  ]}
 AnimGraph {nodes [
   n1 = "Constraint - upperarmDn_l (PP)"!(
    ComponentPose=Local To Component(LocalPose=Copy Pose From Mesh() {bUseAttachedParent=true})
    Alpha=1
    ConstraintWeights_0=0.5
    ConstraintWeights_1:float=0.5
   ) -> (Pose) {
    alphaBoolBlend {blendOption=Linear}
    boneToModify {boneName=upperarmDn_l}
    constraintSetup [{targetBone {boneName=clavicle_l}}, {targetBone {boneName=spine_05}}]
    constraintWeights [1,1]
   }
   n2 = "Constraint - upperarmDn_r (PP)"!(ComponentPose=@n1, Alpha=1, ConstraintWeights_0=0.5, ConstraintWeights_1:float=0.5) -> (Pose) {
    alphaBoolBlend {blendOption=Linear}
    boneToModify {boneName=upperarmDn_r}
    constraintSetup [{targetBone {boneName=clavicle_r}}, {targetBone {boneName=spine_05}}]
    constraintWeights [1,1]
   }
   n3 = "Constraint - thigh_l_fix_a (PP)"!(ComponentPose=@n2, Alpha=1, ConstraintWeights_0=0.5, ConstraintWeights_1:float=0.5) -> (Pose) {
    alphaBoolBlend {blendOption=Linear}
    boneToModify {boneName=thigh_l_fix_a}
    constraintSetup [{targetBone {boneName=thigh_twist_01_l}}, {targetBone {boneName=pelvis}}]
    constraintWeights [1,1]
   }
   n4 = "Constraint - thigh_r_fix_a (PP)"!(ComponentPose=@n3, Alpha=1, ConstraintWeights_0=0.5, ConstraintWeights_1:float=0.5) -> (Pose) {
    alphaBoolBlend {blendOption=Linear}
    boneToModify {boneName=thigh_r_fix_a}
    constraintSetup [{targetBone {boneName=thigh_twist_01_r}}, {targetBone {boneName=pelvis}}]
    constraintWeights [1,1]
   }
   n5 = "Constraint - thigh_volume_bk_l (PP)"!(ComponentPose=@n4, Alpha=1, ConstraintWeights_0=0.5, ConstraintWeights_1:float=0.5) -> (Pose) {
    alphaBoolBlend {blendOption=Linear}
    boneToModify {boneName=thigh_volume_bk_l}
    constraintSetup [{targetBone {boneName=thigh_twist_01_l}}, {targetBone {boneName=pelvis}}]
    constraintWeights [1,1]
   }
   n6 = "Constraint - thigh_volume_bk_r (PP)"!(ComponentPose=@n5, Alpha=1, ConstraintWeights_0=0.5, ConstraintWeights_1:float=0.5) -> (Pose) {
    alphaBoolBlend {blendOption=Linear}
    boneToModify {boneName=thigh_volume_bk_r}
    constraintSetup [{targetBone {boneName=thigh_twist_01_r}}, {targetBone {boneName=pelvis}}]
    constraintWeights [1,1]
   }
   n7 = Get leg_rot_l() -> (leg_rot_l:double)
   n8 = "Transform (Modify) Bone - Bone: thigh_l_fix_a"(ComponentPose=@n6, Alpha=@n7, Translation="0,30,0", Rotation="0,0,0", Scale="1,1,1") {
    rotationMode=BMM_Additive
    rotationSpace=BCS_BoneSpace
    scaleMode=BMM_Additive
    scaleSpace=BCS_BoneSpace
    translationMode=BMM_Additive
    translationSpace=BCS_BoneSpace
    alphaBoolBlend {blendOption=Linear}
    boneToModify {boneName=thigh_l_fix_a}
    translation {y=30}
   }
   n9 = Get leg_rot_r() -> (leg_rot_r:double)
   n10 = "Transform (Modify) Bone - Bone: thigh_r_fix_a"(ComponentPose=@n8, Alpha=@n9, Translation="0,-30,0", Rotation="0,0,0", Scale="1,1,1") {
    rotationMode=BMM_Additive
    rotationSpace=BCS_BoneSpace
    scaleMode=BMM_Additive
    scaleSpace=BCS_BoneSpace
    translationMode=BMM_Additive
    translationSpace=BCS_BoneSpace
    alphaBoolBlend {blendOption=Linear}
    boneToModify {boneName=thigh_r_fix_a}
    translation {y=-30}
   }
   n11 = "Transform (Modify) Bone - Bone: thigh_volume_bk_l"(ComponentPose=@n10, Alpha=@n7, Translation="2,-5,0", Rotation="0,0,0", Scale="1,1,1") {
    rotationMode=BMM_Additive
    rotationSpace=BCS_BoneSpace
    scaleMode=BMM_Additive
    scaleSpace=BCS_BoneSpace
    translationMode=BMM_Additive
    translationSpace=BCS_BoneSpace
    alphaBoolBlend {blendOption=Linear}
    boneToModify {boneName=thigh_volume_bk_l}
    translation=<2 -5>
   }
   n12 = "Transform (Modify) Bone - Bone: thigh_volume_bk_r"(ComponentPose=@n11, Alpha=@n9, Translation="2,-5,0", Rotation="0,0,0", Scale="1,1,1") {
    rotationMode=BMM_Additive
    rotationSpace=BCS_BoneSpace
    scaleMode=BMM_Additive
    scaleSpace=BCS_BoneSpace
    translationMode=BMM_Additive
    translationSpace=BCS_BoneSpace
    alphaBoolBlend {blendOption=Linear}
    boneToModify {boneName=thigh_volume_bk_r}
    translation=<2 -5>
   }
   n13 = "Transform (Modify) Bone - Bone: upperarmDn_r"(ComponentPose=@n12, Alpha=Get Shoulder_rot_up_r() -> (Shoulder_rot_up_r:double), Translation="0,0,-10", Rotation="0,0,0", Scale="1,1,1") {
    rotationMode=BMM_Additive
    rotationSpace=BCS_BoneSpace
    scaleMode=BMM_Additive
    scaleSpace=BCS_BoneSpace
    translationMode=BMM_Additive
    translationSpace=BCS_BoneSpace
    alphaBoolBlend {blendOption=Linear}
    boneToModify {boneName=upperarmDn_r}
    translation {z=-10}
   }
   n14 = "Transform (Modify) Bone - Bone: upperarmDn_l"(ComponentPose=@n13, Alpha=Get Shoulder_rot_up_l() -> (Shoulder_rot_up_l:double), Translation="0,0,10", Rotation="0,0,0", Scale="1,1,1") {
    rotationMode=BMM_Additive
    rotationSpace=BCS_BoneSpace
    scaleMode=BMM_Additive
    scaleSpace=BCS_BoneSpace
    translationMode=BMM_Additive
    translationSpace=BCS_BoneSpace
    alphaBoolBlend {blendOption=Linear}
    boneToModify {boneName=upperarmDn_l}
    translation {z=10}
   }
   n15 = "Transform (Modify) Bone - Bone: body_pumice_l_01"(
    ComponentPose=@n14
    Alpha=1
    Translation=Get ShardFloatA() -> (ShardFloatA:Vector)
    Rotation=Get ShardFloat_rot_A() -> (ShardFloat_rot_A:Rotator)
    Scale="1,1,1"
   ) {translationMode=BMM_Additive, translationSpace=BCS_BoneSpace, alphaBoolBlend {blendOption=Linear}, boneToModify {boneName=body_pumice_l_01}}
   n16 = "Transform (Modify) Bone - Bone: body_pumice_l_02"(
    ComponentPose=@n15
    Alpha=1
    Translation=Get ShardFloatB() -> (ShardFloatB:Vector)
    Rotation=Get ShardFloat_rot_B() -> (ShardFloat_rot_B:Rotator)
    Scale="1,1,1"
   ) {translationMode=BMM_Additive, translationSpace=BCS_BoneSpace, alphaBoolBlend {blendOption=Linear}, boneToModify {boneName=body_pumice_l_02}}
   n17 = "Transform (Modify) Bone - Bone: body_pumice_r_01"(
    ComponentPose=@n16
    Alpha=1
    Translation=Get ShardFloatC() -> (ShardFloatC:Vector)
    Rotation=Get ShardFloat_rot_C() -> (ShardFloat_rot_C:Rotator)
    Scale="1,1,1"
   ) {translationMode=BMM_Additive, translationSpace=BCS_BoneSpace, alphaBoolBlend {blendOption=Linear}, boneToModify {boneName=body_pumice_r_01}}
   n18 = RigidBody(ComponentPose=@n17) {
    bClampLinearTranslationLimitToRefPose=true
    bOverrideWorldGravity=true
    overridePhysicsAsset=&/M_MED_Tower_Sentinel_Physics
    simulationSpace=BaseBoneSpace
    alphaBoolBlend {blendOption=Linear}
    baseBoneRef {boneName=pelvis}
    componentAppliedLinearAccClamp=<1000 1000 1000>
    componentLinearAccScale=<0.35 0.45 0.45>
    componentLinearVelScale=<0.3 0.3 0.3>
    overrideWorldGravity {z=-980}
   }
   Output Pose(Result=Component To Local(ComponentPose=@n18)) {layerGroup=None}
  ]}
 EventGraph {nodes [
   Event BlueprintUpdateAnimation()
   > n1 = Sequence!(execute) -> (then_0, then_1, then_2:Exec, then_3:Exec, then_4:Exec, then_5:Exec, then_6:Exec, then_7:Exec)
   n2 = Get Skel_Mesh()
   > Set leg_rot_r(leg_rot_r:double=GetAlphaFromReferenceTransformDelta(SkeletalMeshComponent=@n2, BoneName=thigh_r, BaseName=pelvis, InputRange=(X=0,Y=-180), TwistAxis="0,0,1")) -> (Output_Get:double)
   > Set leg_rot_l(leg_rot_l:double=GetAlphaFromReferenceTransformDelta(SkeletalMeshComponent=@n2, BoneName=thigh_twist_01_l, BaseName=pelvis, InputRange=(X=0,Y=-150), TwistAxis="0,0,1")) -> (Output_Get:double)
   n3 = Get Skel_Mesh()
   > Set Shoulder_rot_up_r(Shoulder_rot_up_r:double=GetAlphaFromReferenceTransformDelta(SkeletalMeshComponent=@n3, BoneName=upperarm_r, BaseName=clavicle_r, DeltaComponentType=EulerY, InputRange=(X=0,Y=90), TwistAxis="0,0,1")) -> (Output_Get:double)
   > Set Shoulder_rot_up_l(Shoulder_rot_up_l:double=GetAlphaFromReferenceTransformDelta(SkeletalMeshComponent=@n3, BoneName=upperarm_l, BaseName=clavicle_l, DeltaComponentType=EulerY, InputRange=(X=0,Y=90), TwistAxis="0,0,1")) -> (Output_Get:double)
   n4 = AmbientMotion!(
    execute=@n1.then_1
    self
    Period:Vector="1.5,1,2"
    PeriodMult:double=0.9
    TimeOffset:double=0
    Magnitude:Vector="0.7,0.56,0.22"
    MagnitudeMult:double=0.7
    RotateMult:double=10
   ) -> (then, VectorOut:Vector, RotateOut:Rotator) {is=title K2Node_CallFunction}
   > Set ShardFloatA(ShardFloatA:Vector=@n4) -> (Output_Get:Vector)
   > Set ShardFloat_rot_A(ShardFloat_rot_A:Rotator=@n4.RotateOut) -> (Output_Get:Rotator)
   n5 = AmbientMotion(execute=@n1.then_2, Period="1.5,1,2", PeriodMult=0.9, TimeOffset=0, Magnitude="0.7,0.56,0.22", MagnitudeMult=0.7, RotateMult=10)
   > Set ShardFloatB(ShardFloatB:Vector=@n5) -> (Output_Get:Vector)
   > Set ShardFloat_rot_B(ShardFloat_rot_B:Rotator=@n5.RotateOut) -> (Output_Get:Rotator)
   n6 = AmbientMotion(execute=@n1.then_3, Period="1.5,1,2", PeriodMult=0.9, TimeOffset=0, Magnitude="0.7,0.56,0.22", MagnitudeMult=0.7, RotateMult=10)
   > Set ShardFloatC(ShardFloatC:Vector=@n6) -> (Output_Get:Vector)
   > Set ShardFloat_rot_C(ShardFloat_rot_C:Rotator=@n6.RotateOut) -> (Output_Get:Rotator)
   Event BlueprintInitializeAnimation()
   > Set Skel_Mesh(Skel_Mesh=GetOwnerSkeletalMeshComponent())
  ]}
]
variables [
 leg_rot_l {type=number}
 leg_rot_r {type=number}
 ShardFloat_rot_A {type=Rotator}
 ShardFloat_rot_B {type=Rotator}
 ShardFloat_rot_C {type=Rotator}
 ShardFloatA {type=Vector}
 ShardFloatB {type=Vector}
 ShardFloatC {type=Vector}
 Shoulder_rot_up_l {type=number}
 Shoulder_rot_up_r {type=number}
 Skel_Mesh {type=SkeletalMeshComponent}
]
