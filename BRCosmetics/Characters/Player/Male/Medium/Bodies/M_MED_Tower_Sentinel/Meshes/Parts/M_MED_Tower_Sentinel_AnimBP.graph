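# Anim Blueprint for the Tower Sentinel body part that makes eight pumice shards around the head float and wobble.
# Each update, AmbientMotion turns game time into per-axis sine offsets and rotations, tuned per shard and stored in variables.
# The AnimGraph copies the attached parent mesh's pose, rotation-constrains five shards between head and spine_05 with weighted blends, then adds the offsets to all eight in bone space.
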AnimBlueprint /BRCosmetics/Characters/Player/Male/Medium/Bodies/M_MED_Tower_Sentinel/Meshes/Parts/M_MED_Tower_Sentinel_AnimBP
target_skeleton=&/M_MED_Tower_Sentinel_Skeleton
graphs [
 AmbientMotion {nodes [
   n1 = AmbientMotion!() -> (
    then
    Period:Vector="1.5,1,2"
    PeriodMult:double=1
    TimeOffset:double=~
    Magnitude:Vector="0.3,0.56,0.22"
    MagnitudeMult:double=1
    RotateMult:double=1
   ) {is=title K2Node_FunctionEntry}
   n2 = BreakVector(InVec=@n1.Magnitude)
   n3 = BreakVector(InVec=@n1)
   n4 = GetGameTimeInSeconds()
   n5 = float * float(A=float * float(A=@n1.PeriodMult, B=@n3.Y), B=@n4)
   n6 = float * float(A="Sin (Radians)"(A=float + float(A=@n5, B=@n1.TimeOffset) {is=title K2Node_CommutativeAssociativeBinaryOperator}), B=@n2.Y)
   n7 = float * float(A=float * float(A=@n1.PeriodMult, B=@n3.Z), B=@n4)
   n8 = float * float(A="Sin (Radians)"(A=float + float(A=@n7, B=@n1.TimeOffset)), B=@n2.Z)
   n9 = float * float(A=float * float(A=@n1.PeriodMult, B=@n3), B=@n4)
   n10 = float * float(A="Sin (Radians)"(A=float + float(A=@n9, B=@n1.TimeOffset)), B=@n2)
   n11 = MakeRotator(Roll=@n10, Pitch=@n6, Yaw=@n8)
   > Return Node!(
    execute
    VectorOut:Vector=vector * float(A=MakeVector(X=@n10, Y=@n6, Z=@n8), B=@n1.MagnitudeMult)
    RotateOut:Rotator="Scale Rotator (Float)"!(A:Rotator="Scale Rotator (Float)"(A=@n11, B=@n1.MagnitudeMult), B:float=@n1.RotateMult) -> (ReturnValue:Rotator)
   )
  ]}
 AnimGraph {nodes [
   n1 = "Constraint - head_pumice_l_01 (RR)"!(
    ComponentPose=Local To Component(LocalPose=Copy Pose From Mesh() {bUseAttachedParent=true})
    Alpha=1
    ConstraintWeights_0=0.3
    ConstraintWeights_1:float=0.7
   ) -> (Pose) {
    alphaBoolBlend {blendOption=Linear}
    boneToModify {boneName=head_pumice_l_01}
    constraintSetup [{transformType=Rotation, targetBone {boneName=head}}, {transformType=Rotation, targetBone {boneName=spine_05}}]
    constraintWeights [1,1]
   }
   n2 = "Constraint - head_pumice_l_02 (RR)"!(ComponentPose=@n1, Alpha=1, ConstraintWeights_0=0.2, ConstraintWeights_1:float=0.8) -> (Pose) {
    alphaBoolBlend {blendOption=Linear}
    boneToModify {boneName=head_pumice_l_02}
    constraintSetup [{transformType=Rotation, targetBone {boneName=head}}, {transformType=Rotation, targetBone {boneName=spine_05}}]
    constraintWeights [1,1]
   }
   n3 = "Constraint - head_pumice_l_03 (RR)"!(ComponentPose=@n2, Alpha=1, ConstraintWeights_0=0.2, ConstraintWeights_1:float=0.8) -> (Pose) {
    alphaBoolBlend {blendOption=Linear}
    boneToModify {boneName=head_pumice_l_03}
    constraintSetup [{transformType=Rotation, targetBone {boneName=head}}, {transformType=Rotation, targetBone {boneName=spine_05}}]
    constraintWeights [1,1]
   }
   n4 = "Constraint - head_pumice_l_07 (RR)"!(ComponentPose=@n3, Alpha=1, ConstraintWeights_0=0.2, ConstraintWeights_1:float=0.8) -> (Pose) {
    alphaBoolBlend {blendOption=Linear}
    boneToModify {boneName=head_pumice_l_07}
    constraintSetup [{transformType=Rotation, targetBone {boneName=head}}, {transformType=Rotation, targetBone {boneName=spine_05}}]
    constraintWeights [1,1]
   }
   n5 = "Constraint - head_pumice_l_08 (RR)"!(ComponentPose=@n4, Alpha=1, ConstraintWeights_0=0.5, ConstraintWeights_1:float=0.5) -> (Pose) {
    alphaBoolBlend {blendOption=Linear}
    boneToModify {boneName=head_pumice_l_08}
    constraintSetup [{transformType=Rotation, targetBone {boneName=head}}, {transformType=Rotation, targetBone {boneName=spine_05}}]
    constraintWeights [1,1]
   }
   n6 = "Transform (Modify) Bone - Bone: head_pumice_l_01"(
    ComponentPose=@n5
    Alpha=1
    Translation=Get ShardFloatD() -> (ShardFloatD:Vector)
    Rotation=Get ShardFloat_rot_D() -> (ShardFloat_rot_D:Rotator)
    Scale="1,1,1"
   ) {translationMode=BMM_Additive, translationSpace=BCS_BoneSpace, alphaBoolBlend {blendOption=Linear}, boneToModify {boneName=head_pumice_l_01}}
   n7 = "Transform (Modify) Bone - Bone: head_pumice_l_02"(
    ComponentPose=@n6
    Alpha=1
    Translation=Get ShardFloatE() -> (ShardFloatE:Vector)
    Rotation=Get ShardFloat_rot_E() -> (ShardFloat_rot_E:Rotator)
    Scale="1,1,1"
   ) {translationMode=BMM_Additive, translationSpace=BCS_BoneSpace, alphaBoolBlend {blendOption=Linear}, boneToModify {boneName=head_pumice_l_02}}
   n8 = "Transform (Modify) Bone - Bone: head_pumice_l_03"(
    ComponentPose=@n7
    Alpha=1
    Translation=Get ShardFloatF() -> (ShardFloatF:Vector)
    Rotation=Get ShardFloat_rot_F() -> (ShardFloat_rot_F:Rotator)
    Scale="1,1,1"
   ) {translationMode=BMM_Additive, translationSpace=BCS_BoneSpace, alphaBoolBlend {blendOption=Linear}, boneToModify {boneName=head_pumice_l_03}}
   n9 = "Transform (Modify) Bone - Bone: head_pumice_l_04"(
    ComponentPose=@n8
    Alpha=1
    Translation=Get ShardFloatA() -> (ShardFloatA:Vector)
    Rotation=Get ShardFloat_rot_A() -> (ShardFloat_rot_A:Rotator)
    Scale="1,1,1"
   ) {translationMode=BMM_Additive, translationSpace=BCS_BoneSpace, alphaBoolBlend {blendOption=Linear}, boneToModify {boneName=head_pumice_l_04}}
   n10 = "Transform (Modify) Bone - Bone: head_pumice_l_05"(
    ComponentPose=@n9
    Alpha=1
    Translation=Get ShardFloatB() -> (ShardFloatB:Vector)
    Rotation=Get ShardFloat_rot_B() -> (ShardFloat_rot_B:Rotator)
    Scale="1,1,1"
   ) {translationMode=BMM_Additive, translationSpace=BCS_BoneSpace, alphaBoolBlend {blendOption=Linear}, boneToModify {boneName=head_pumice_l_05}}
   n11 = "Transform (Modify) Bone - Bone: head_pumice_l_06"(
    ComponentPose=@n10
    Alpha=1
    Translation=Get ShardFloatC() -> (ShardFloatC:Vector)
    Rotation=Get ShardFloat_rot_C() -> (ShardFloat_rot_C:Rotator)
    Scale="1,1,1"
   ) {translationMode=BMM_Additive, translationSpace=BCS_BoneSpace, alphaBoolBlend {blendOption=Linear}, boneToModify {boneName=head_pumice_l_06}}
   n12 = "Transform (Modify) Bone - Bone: head_pumice_l_07"(
    ComponentPose=@n11
    Alpha=1
    Translation=Get ShardFloatG() -> (ShardFloatG:Vector)
    Rotation=Get ShardFloat_rot_G() -> (ShardFloat_rot_G:Rotator)
    Scale="1,1,1"
   ) {translationMode=BMM_Additive, translationSpace=BCS_BoneSpace, alphaBoolBlend {blendOption=Linear}, boneToModify {boneName=head_pumice_l_07}}
   n13 = "Transform (Modify) Bone - Bone: head_pumice_l_08"(
    ComponentPose=@n12
    Alpha=1
    Translation=Get ShardFloatH() -> (ShardFloatH:Vector)
    Rotation=Get ShardFloat_rot_H() -> (ShardFloat_rot_H:Rotator)
    Scale="1,1,1"
   ) {translationMode=BMM_Additive, translationSpace=BCS_BoneSpace, alphaBoolBlend {blendOption=Linear}, boneToModify {boneName=head_pumice_l_08}}
   Output Pose(Result=Component To Local(ComponentPose=@n13)) {layerGroup=None}
  ]}
 EventGraph {nodes [
   Event BlueprintUpdateAnimation()
   > n1 = Sequence!(execute) -> (then_0, then_1, then_2:Exec, then_3:Exec, then_4:Exec, then_5:Exec, then_6:Exec, then_7:Exec)
   > n2 = AmbientMotion!(
    execute
    self
    Period:Vector="1.5,1,2"
    PeriodMult:double=0.9
    TimeOffset:double=0
    Magnitude:Vector="0.7,0.56,0.22"
    MagnitudeMult:double=0.7
    RotateMult:double=10
   ) -> (then, VectorOut:Vector, RotateOut:Rotator) {is=title K2Node_CallFunction}
   > Set ShardFloatA(ShardFloatA:Vector=@n2) -> (Output_Get:Vector)
   > Set ShardFloat_rot_A(ShardFloat_rot_A:Rotator=@n2.RotateOut) -> (Output_Get:Rotator)
   n3 = AmbientMotion(execute=@n1.then_1, Period="1.5,1,2", PeriodMult=0.7, TimeOffset=0, Magnitude="0.7,0.56,0.22", MagnitudeMult=0.85, RotateMult=12)
   > Set ShardFloatB(ShardFloatB:Vector=@n3) -> (Output_Get:Vector)
   > Set ShardFloat_rot_B(ShardFloat_rot_B:Rotator=@n3.RotateOut) -> (Output_Get:Rotator)
   n4 = AmbientMotion(execute=@n1.then_2, Period="1.5,1,2", PeriodMult=0.8, TimeOffset=0, Magnitude="0.7,0.56,0.22", MagnitudeMult=0.8, RotateMult=14)
   > Set ShardFloatC(ShardFloatC:Vector=@n4) -> (Output_Get:Vector)
   > Set ShardFloat_rot_C(ShardFloat_rot_C:Rotator=@n4.RotateOut) -> (Output_Get:Rotator)
   n5 = AmbientMotion(execute=@n1.then_3, Period="1.5,1,2", PeriodMult=0.9, TimeOffset=0, Magnitude="0.7,0.56,0.22", MagnitudeMult=0.8, RotateMult=16)
   > Set ShardFloatD(ShardFloatD:Vector=@n5) -> (Output_Get:Vector)
   > Set ShardFloat_rot_D(ShardFloat_rot_D:Rotator=@n5.RotateOut) -> (Output_Get:Rotator)
   n6 = AmbientMotion(execute=@n1.then_4, Period="1.5,1,2", PeriodMult=0.8, TimeOffset=0, Magnitude="0.7,0.56,0.22", MagnitudeMult=0.8, RotateMult=14)
   > Set ShardFloatE(ShardFloatE:Vector=@n6) -> (Output_Get:Vector)
   > Set ShardFloat_rot_E(ShardFloat_rot_E:Rotator=@n6.RotateOut) -> (Output_Get:Rotator)
   n7 = AmbientMotion(execute=@n1.then_5, Period="1.5,1,2", PeriodMult=0.75, TimeOffset=0, Magnitude="0.7,0.56,0.22", MagnitudeMult=0.8, RotateMult=12)
   > Set ShardFloatF(ShardFloatF:Vector=@n7) -> (Output_Get:Vector)
   > Set ShardFloat_rot_F(ShardFloat_rot_F:Rotator=@n7.RotateOut) -> (Output_Get:Rotator)
   n8 = AmbientMotion(execute=@n1.then_6, Period="1.5,1,2", PeriodMult=0.85, TimeOffset=0, Magnitude="0.7,0.56,0.22", MagnitudeMult=0.8, RotateMult=14)
   > Set ShardFloatG(ShardFloatG:Vector=@n8) -> (Output_Get:Vector)
   > Set ShardFloat_rot_G(ShardFloat_rot_G:Rotator=@n8.RotateOut) -> (Output_Get:Rotator)
   n9 = AmbientMotion(execute=@n1.then_7, Period="1.5,1,2", PeriodMult=0.7, TimeOffset=0, Magnitude="0.7,0.56,0.22", MagnitudeMult=0.8, RotateMult=16)
   > Set ShardFloatH(ShardFloatH:Vector=@n9) -> (Output_Get:Vector)
   > Set ShardFloat_rot_H(ShardFloat_rot_H:Rotator=@n9.RotateOut) -> (Output_Get:Rotator)
   Reroute Node()
   n10 = GetGameTimeInSeconds()
   n11 = "Sin (Radians)"(A=float * float(A=float * float(A=1, B=1.5), B=@n10))
   n12 = "Sin (Radians)"(A=float * float(A=float * float(A=1, B=2.5), B=@n10))
   n13 = "Sin (Radians)"(A=float * float(A=float * float(A=1, B=2), B=@n10))
   > Set ShardFloatA(ShardFloatA=MakeVector(X=float * float(A=@n13, B=0.65), Y=float * float(A=@n11, B=0.86), Z=float * float(A=@n12, B=0.66)))
   n14 = "Cos (Radians)"!(A:double=float * float(A=float * float(A=1, B=1.5), B=@n10)) -> (ReturnValue:double)
   n15 = "Cos (Radians)"(A=float * float(A=float * float(A=1, B=1.78), B=@n10))
   n16 = "Cos (Radians)"(A=float * float(A=float * float(A=1, B=1.15), B=@n10))
   > Set ShardFloatB(ShardFloatB=MakeVector(X=float * float(A=@n14, B=0.6), Y=float * float(A=@n15, B=0.65), Z=float * float(A=@n16, B=0.75)))
   n17 = "Sin (Radians)"(A=float * float(A=float * float(A=1, B=1.5), B=@n10))
   n18 = "Sin (Radians)"(A=float * float(A=float * float(A=1, B=1), B=@n10))
   n19 = "Sin (Radians)"(A=float * float(A=float * float(A=1, B=2), B=@n10))
   > Set ShardFloatC(ShardFloatC=MakeVector(X=float * float(A=@n17, B=0.55), Y=float * float(A=@n18, B=0.8), Z=float * float(A=@n19, B=0.6)))
  ]}
]
variables [
 ShardFloat_rot_A {type=Rotator}
 ShardFloat_rot_B {type=Rotator}
 ShardFloat_rot_C {type=Rotator}
 ShardFloat_rot_D {type=Rotator}
 ShardFloat_rot_E {type=Rotator}
 ShardFloat_rot_F {type=Rotator}
 ShardFloat_rot_G {type=Rotator}
 ShardFloat_rot_H {type=Rotator}
 ShardFloatA {type=Vector}
 ShardFloatB {type=Vector}
 ShardFloatC {type=Vector}
 ShardFloatD {type=Vector}
 ShardFloatE {type=Vector}
 ShardFloatF {type=Vector}
 ShardFloatG {type=Vector}
 ShardFloatH {type=Vector}
]
